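# Editor utility tool that reparents selected terrain material instances to the Figment environment's cliff material instance.
# ReparentMaterialInstances loops over the selected assets, keeping material instance constants parented directly to Athena_Terrain_Mat_Biome.
# It sets their parent to Athena_Terrain_Mat_Biome_Figment_Inst and collects them in assetsToSave, which is cleared at the start of each run.
# A second Sequence step then checks out and saves the collected assets.
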EditorUtilityBlueprint /FigmentEnvironment/Tools/ReparentMaterials
graphs [
 EventGraph {links [], nodes []}
 ReparentMaterialInstances {nodes [
   ReparentMaterialInstances() {is=title K2Node_FunctionEntry}
   > n1 = Sequence()
   > Clear(TargetArray:Array of Object References=Get assetsToSave() -> (assetsToSave:Array of Object References))
   > n2 = GetSelectedAssets()
   > n3 = For Each Loop(Array:Array of Object References=@n2) -> (Array Element:Object Reference)
   > n4 = Cast To MaterialInstanceConstant(Object=@n3)
   > Branch(Condition="Equal (Object)"(
     A=Get Parent!(self:Material Instance Object Reference=@n4) -> (Parent:Material Interface Object Reference)
     B=/Game/Athena/Environments/Landscape/Material/Athena_Terrain_Mat_Biome
    ))
   > SetMaterialInstanceParent!(
    execute
    Instance:Material Instance Constant Object Reference=@n4
    NewParent:Material Interface Object Reference=/FigmentEnvironment/Materials/Cliff/Athena_Terrain_Mat_Biome_Figment_Inst
   ) -> (then)
   > Add(TargetArray:Array of Object References=Get assetsToSave(), NewItem:"Object Reference (by ref)"=@n4) {is=title K2Node_CallArrayFunction}
   n5 = Get assetsToSave()
   CheckoutLoadedAssets(execute=@n1.then_1, AssetsToCheckout=@n5)
   > SaveLoadedAssets(AssetsToSave=@n5, bOnlyIfIsDirty=false)
   PrintString(InString=GetDisplayName(Object=@n3))
  ]}
]
variables [assetsToSave {type="array<Object>", default []}]
